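# Sound cue for the drum kit of a Creative toy keyboard, mapping two octaves of keys to 24 one-shot samples.
# The KeysIndex integer parameter drives a switch whose default slot is empty, with natural keys ordered before sharps.
# All quality levels share that switch.
# The cue primes on load, uses Music Blocks attenuation and caps concurrency, stopping the oldest voice.
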SoundCue /Game/Sounds/Creative/Toys/Keyboard/Drums/CM_Keyboard_Drums_Cue
root=QualityLevel_0
nodes [
 {id=QualityLevel_0, children [Switch_1,Switch_1,Switch_1,Switch_1]}
 {
  id=Switch_1
  children [
   null
   WavePlayer_0
   WavePlayer_2
   WavePlayer_4
   WavePlayer_6
   WavePlayer_8
   WavePlayer_10
   WavePlayer_12
   WavePlayer_13
   WavePlayer_15
   WavePlayer_17
   WavePlayer_20
   WavePlayer_21
   WavePlayer_23
   WavePlayer_25
   WavePlayer_1
   WavePlayer_3
   WavePlayer_7
   WavePlayer_9
   WavePlayer_11
   WavePlayer_14
   WavePlayer_16
   WavePlayer_19
   WavePlayer_22
   WavePlayer_24
  ]
  properties {intParameterName=KeysIndex}
 }
 {id=WavePlayer_0, properties {soundWaveAssetPtr=&/CM_Drums_01_C1_Kick01}}
 {id=WavePlayer_1, properties {soundWaveAssetPtr=&/CM_Drums_02_C_1_Kick02}}
 {id=WavePlayer_10, properties {soundWaveAssetPtr=&/CM_Drums_10_A1_HiHatC02}}
 {id=WavePlayer_11, properties {soundWaveAssetPtr=&/CM_Drums_11_A_1_HiHatC03}}
 {id=WavePlayer_12, properties {soundWaveAssetPtr=&/CM_Drums_12_B1_HiHatO01}}
 {id=WavePlayer_13, properties {soundWaveAssetPtr=&/CM_Drums_13_C2_HiHatO02}}
 {id=WavePlayer_14, properties {soundWaveAssetPtr=&/CM_Drums_14_C_2_Ride01}}
 {id=WavePlayer_15, properties {soundWaveAssetPtr=&/CM_Drums_15_D2_Crash01}}
 {id=WavePlayer_16, properties {soundWaveAssetPtr=&/CM_Drums_16_D_2_Clap01}}
 {id=WavePlayer_17, properties {soundWaveAssetPtr=&/CM_Drums_17_E2_Clap02}}
 {id=WavePlayer_19, properties {soundWaveAssetPtr=&/CM_Drums_19_F_2_Rim02}}
 {id=WavePlayer_2, properties {soundWaveAssetPtr=&/CM_Drums_03_D1_Kick03}}
 {id=WavePlayer_20, properties {soundWaveAssetPtr=&/CM_Drums_18_F2_Rim01}}
 {id=WavePlayer_21, properties {soundWaveAssetPtr=&/CM_Drums_20_G2_Perc01}}
 {id=WavePlayer_22, properties {soundWaveAssetPtr=&/CM_Drums_21_G_2_Perc02}}
 {id=WavePlayer_23, properties {soundWaveAssetPtr=&/CM_Drums_22_A2_Tom01}}
 {id=WavePlayer_24, properties {soundWaveAssetPtr=&/CM_Drums_23_A_2_Tom02}}
 {id=WavePlayer_25, properties {soundWaveAssetPtr=&/CM_Drums_24_B2_Tom03}}
 {id=WavePlayer_3, properties {soundWaveAssetPtr=&/CM_Drums_04_D_1_Kick04}}
 {id=WavePlayer_4, properties {soundWaveAssetPtr=&/CM_Drums_05_E1_Snare01}}
 {id=WavePlayer_6, properties {soundWaveAssetPtr=&/CM_Drums_06_F1_Snare02}}
 {id=WavePlayer_7, properties {soundWaveAssetPtr=&/CM_Drums_07_F_1_Snare03}}
 {id=WavePlayer_8, properties {soundWaveAssetPtr=&/CM_Drums_08_G1_Snare04}}
 {id=WavePlayer_9, properties {soundWaveAssetPtr=&/CM_Drums_09_G_1_HiHatC01}}
]
properties {
 attenuationSettings=/Game/Sounds/Creative/Attenuations/Fort_Creative_MusicBlocks_Att
 bPrimeOnLoad=true
 maxDistance=100000
 soundClassObject=&SFX
 volumeMultiplier=1.8
 concurrencyOverrides {maxCount=12, resolutionRule=StopOldest}
}
